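# Sound cue for the GrillMoon event that plays the circle key ignite sound effect.
# A single wave player feeds the output directly, with no randomization or modulation.
# It uses the close point-of-interest attenuation preset and routes through the Fort_Events sound class.
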SoundCue /GrillMoon_Audio/Assets/tdm/SC_GrillMoon_tdm_Circle_Key_Ignite
root=WavePlayer_0
nodes [{id=WavePlayer_0, properties {soundWaveAssetPtr=/GrillMoon_Audio/Assets/wav/GrillMoon_Circle_Key_Ignite}}]
properties {
 attenuationSettings=/GrillMoon_Audio/Assets/Settings/ATT_GrillMoon_POI_Close
 maxDistance=14000
 soundClassObject=&Fort_Events
 volumeMultiplier=1
}
